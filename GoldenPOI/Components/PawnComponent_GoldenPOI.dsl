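; Pawn component for the Golden POI feature that caches references to the feature's manager and to the owning player pawn.
; InitializeManagerReference finds all B_GoldenPOI_Manager_V2 actors in the world and stores the first one, if any exists, as GoldenPOIManager.
; InitializeOwnerReference casts the component's owner to FortPlayerPawn and stores it as OwnerPawn.
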(fn InitializeManagerReference
  (bind _outactors (Actor|GetAllActorsOfClass "/GoldenPOI/Gameplay/B_GoldenPOI_Manager_V2.B_GoldenPOI_Manager_V2_C"))
  (if (Utilities|Array|IsNotEmpty _outactors)
    (Variables|Default|SetGoldenPOIManager (Utilities|Array|Get(acopy) _outactors))))

(fn InitializeOwnerReference
  (bind _asfort_player_pawn (Utilities|Casting|CastToFortPlayerPawn (Components|GetOwner)))
  (Variables|Default|SetOwnerPawn _asfort_player_pawn))
